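# Translucent, unlit, two-sided Niagara mesh particle material for the green lines in the Candor emote, with particle color driving emissive.
# Default mask combines sine-and-floor stripes in U and animated bands in V with edge fades and a panned gradient.
# Scalar parameters blend toward an Atlas scanline mask or a simple ray mask.
# Dynamic parameters drive band speed, gradient offset and an appear wipe along V.
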Material /BRCosmetics/Effects/Fort_Effects/Emotes/Candor/M_Candor_Emote_Green_Lines
graph {
 properties {
  bUsedWithNiagaraMeshParticles=true
  blendMode=BLEND_Translucent
  pixelDepthOffsetMode=PDOM_Legacy
  shadingModel=MSM_Unlit
  twoSided=true
  usedShadingModels=Unlit
 }
 nodes [
  n1 = ParticleColor()
  MaterialFunctionCall(Emissive=@n1.RGB) {materialFunction=&MF_FX_EmissiveAdjustments}
  n2 = OneMinus(None=SphereMask(A=TextureCoordinate() {vTiling=0}, B=Constant2Vector() {r=0.5}) {attenuationRadius=0.9, hardnessPercent=0})
  n3 = Multiply(A=Saturate(None=Subtract(A=@n2) {constB=0.5})) {constB=5}
  n4 = Constant() {r=0.05}
  n5 = TextureCoordinate()
  n6 = Multiply(A=ParticleColor().A, B=Multiply(A=ComponentMask(None=@n5) {g=true}, B=Add(A=@n3, B=@n4)))
  n7 = ScalarParameter() {parameterName=Use simple ray mask}
  n8 = ComponentMask(None=TextureCoordinate()) {g=true}
  n9 = TextureCoordinate()
  n10 = Power(Base=OneMinus(None=ComponentMask(None=@n9) {g=true}), Exp=ScalarParameter() {defaultValue=1, parameterName=Mask_Edge_power}) {constExponent=1}
  n11 = ScalarParameter() {defaultValue=3, parameterName=Mask Edge intensity}
  n12 = Multiply(A=Saturate(None=Multiply(A=@n8) {constB=5}), B=Saturate(None=Multiply(A=@n10, B=@n11) {constB=3}))
  n13 = MaterialFunctionCall!(
   Number of Scanlines=ScalarParameter() {defaultValue=150, parameterName=Scanline Amount}
   Scanline Speed
   "Scanlines Thickness (Bias)"=ScalarParameter() {defaultValue=-0.15, parameterName=ScanlineSize}
   Scanlines Dark Value=ScalarParameter() {defaultValue=0.5, parameterName=Scanlines Dark}
   Scanlines Light Value=ScalarParameter() {defaultValue=1, parameterName=Scanlines Light}
   Block Opacity
   UVs
   Time
   Block Repeat
  ) {materialFunction=/Game/Effects/Fort_Effects/Effects/Atlas/Materials/MF_Atlas_Scanlines_01}
  n14 = ComponentMask(None=TextureCoordinate()) {g=true}
  n15 = Saturate(None=Multiply(A=Subtract(A=@n14) {constB=0.3}) {constB=5})
  n16 = Power(Base=ComponentMask(None=TextureCoordinate() {vTiling=1.2}) {g=true})
  n17 = Multiply(
   A=Time()
   B=Multiply(A=Constant() {r=0.3}, B=DynamicParameter() {defaultValue=<rgba 0.5 1 0 0>, paramNames [Param1,Param2,appear,speed]}.speed)
  ) {constB=0.3}
  n18 = Add(A=Sine(None=Add(A=@n16, B=@n17)) {period=0.2}) {constB=0.04}
  n19 = TextureCoordinate()
  n20 = Multiply(A=Multiply(A=Floor(None=@n18), B=ComponentMask(None=@n19) {g=true}))
  n21 = Sine(None=ComponentMask(None=TextureCoordinate() {uTiling=50}) {r=true})
  n22 = Multiply(A=Floor(None=Add(A=@n21) {constB=0.1}), B=ComponentMask(None=TextureCoordinate()) {g=true})
  n23 = Add(A=Saturate(None=Multiply(A=@n20)), B=Multiply(A=Multiply(A=@n22) {constB=0.5})) {constB=0}
  n24 = Power(Base=ComponentMask(None=TextureCoordinate()) {g=true}) {constExponent=1}
  n25 = OneMinus(None=ComponentMask(None=TextureCoordinate()) {g=true})
  n26 = Multiply(A=Saturate(None=Multiply(A=@n24) {constB=10}), B=Saturate(None=Multiply(A=@n25) {constB=20}))
  n27 = Multiply(A=Multiply(A=Multiply(A=@n15, B=@n23)), B=Saturate(None=Multiply(A=@n26)))
  n28 = Panner(
   Coordinate=TextureCoordinate()
   Time=LinearInterpolate(Alpha=DynamicParameter() {defaultValue=<rgba 0.5 1 0 0>, paramNames [Param1,Param2,appear,speed]}.Param2) {constA=1, constB=-1.3}
  ) {speedY=0.5}
  n29 = LinearInterpolate(
   A=Multiply(
    A=Multiply(A=@n1.A, B=@n27)
    B=TextureSample(UVs=@n28) {samplerSource=SSM_Clamp_WorldGroupSettings, texture=/Game/Global/Textures/Default/Masks_and_Gradients/gradient_stasis}.R
   )
   B=Multiply(A=ParticleColor().A, B=Multiply(A=@n12, B=@n13.Result))
   Alpha=ScalarParameter() {parameterName=Use Scanlines mask}
  )
  n30 = TextureCoordinate()
  n31 = DynamicParameter() {defaultValue=<rgba 0.5 1 0 0>, paramNames [Param1,Param2,appear,speed]}
  n32 = Multiply(A=Subtract(A=ComponentMask(None=@n30) {g=true}, B=LinearInterpolate(Alpha=@n31.appear) {constA=-0.3}) {constB=-1}) {constB=10}
  Reroute(None=Multiply(A=LinearInterpolate(A=@n29, B=@n6, Alpha=@n7) {constA=1}, B=Saturate(None=@n32)))
 ]
}
