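# Unlit, two-sided translucent Niagara mesh-particle material for the VR visuals of the Candor emote.
# Emissive sums panning rainbow and curve-atlas noise, tinted scrolling pixel squares, and sine-generated blue-green grid lines, all tinted by particle color.
# Opacity combines top/bottom UV masks, particle alpha, an optional brightness-based square mask, view-angle fade and depth fade.
# Niagara dynamic parameters drive scroll time, the vertical appear wipe and an extra color-noise layer.
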Material /BRCosmetics/Effects/Fort_Effects/Emotes/Candor/M_Candor_Emote_VR
graph {
 properties {
  bUsedWithNiagaraMeshParticles=true
  blendMode=BLEND_Translucent
  pixelDepthOffsetMode=PDOM_Legacy
  shadingModel=MSM_Unlit
  twoSided=true
  usedShadingModels=Unlit
 }
 nodes [
  Constant() {r=1}
  n1 = ParticleColor()
  n2 = TextureCoordinate() {uTiling=2}
  n3 = DynamicParameter() {defaultValue=<rgba 0 1 0 1>, paramNames [Color lerp, Param2, appear, speed]}
  n4 = CurveAtlasRowParameter(CurveTime=TextureSample(UVs=Panner(Coordinate=@n2, Time=@n3.speed) {speedY=1}) {samplerType=SAMPLERTYPE_Grayscale, texture=&LowResBlurredNoise}.R) {
   atlas=/Game/Athena/LinearColorCurves/A_Atlas_Turtleneck_Glider
   curve=/Game/Effects/Fort_Effects/Emotes/Candor/Candor_Emote_Curve
   defaultValue=38
  }
  n5 = Multiply(A=@n4)
  n6 = VectorParameter() {parameterName=ColorfulBackground_ColorLerpUV, defaultValue=<rgba 3 0.1 0 0>}
  n7 = Multiply(A=TextureCoordinate(), B=AppendVector(A=@n6.R, B=@n6.R))
  n8 = DynamicParameter() {defaultValue=<rgba 0 1 0 1>, paramNames [Color lerp, Param2, appear, speed]}
  n9 = Desaturation(
   None=TextureSample(UVs=Panner(Coordinate=@n7, Time=@n8.speed) {speedY=1}) {texture=&T_Tile_RainbowColor}.RGB
   Fraction=Constant() {r=-1}
  )
  n10 = Add(A=Round(None=Multiply(A=@n9)), B=@n9)
  n11 = VectorParameter() {parameterName=ColorfulBackground_LerpTexUV, defaultValue=<rgba 1 0.5 0 0>}
  n12 = Panner(
   Coordinate=Multiply(A=TextureCoordinate(), B=AppendVector(A=@n11.R, B=@n11.R))
   Time=DynamicParameter() {defaultValue=<rgba 0 1 0 1>, paramNames [Color lerp, Param2, appear, speed]}.speed
  ) {speedY=2}
  n13 = Power(Base=LinearInterpolate(Alpha=TextureSample(UVs=@n12) {samplerType=SAMPLERTYPE_Grayscale, texture=&LowResBlurredNoise}.R))
  n14 = LinearInterpolate(B=Multiply(A=Add(A=@n5, B=@n10), B=Add(A=@n13) {constB=0}), Alpha=ScalarParameter() {defaultValue=1, parameterName=Use background color})
  n15 = Multiply(A=TextureCoordinate(), B=ComponentMask(None=VectorParameter() {parameterName=Square UV, defaultValue=<rgba 1 0.6 0 0>}.RGB) {g=true, r=true})
  n16 = ComponentMask(None=@n15) {r=true}
  n17 = Power(Base=ComponentMask(None=@n15) {g=true})
  n18 = DynamicParameter() {defaultValue=<rgba 0 1 0 1>, paramNames [Color lerp, Param2, appear, speed]}
  n19 = Time()
  n20 = ScalarParameter() {parameterName=Use Loop square loop panner}
  n21 = Constant2Vector() {g=1.5}
  n22 = Multiply(A=Constant2Vector() {g=1}, B=DynamicParameter() {defaultValue=<rgba 0 1 0 1>, paramNames [Color lerp, Param2, appear, speed]}.speed)
  n23 = ScalarParameter() {parameterName=Use Loop square loop panner}
  n24 = TextureSample(UVs=Panner(
    Coordinate=AppendVector(A=@n16, B=@n17)
    Time=LinearInterpolate(A=@n18.speed, B=@n19, Alpha=@n20)
    Speed=LinearInterpolate(A=@n21, B=@n22, Alpha=@n23)
   ) {speedY=1.5}) {
   samplerType=SAMPLERTYPE_LinearGrayscale
   texture=/Game/UI/Frontend/ItemShop/Com/vfx/T_CommunityVoting_Activities_Pixels.CommunityVoting_Activities_Pixels
  }
  n25 = TextureSample(UVs=Panner(Coordinate=TextureCoordinate() {uTiling=3, vTiling=0}) {speedY=1}) {texture=&T_Tile_RainbowColor}
  n26 = Multiply(
   A=VectorParameter() {parameterName=Square color, defaultValue=<rgba 0 5 4.7451 0>}.RGB
   B=Desaturation(None=Multiply(A=@n24.RGB, B=@n25.RGB), Fraction=Constant() {r=-1})
  )
  n27 = TextureCoordinate()
  n28 = Multiply(A=Subtract(A=ComponentMask(None=@n27) {g=true}) {constB=0.3}) {constB=5}
  n29 = TextureCoordinate()
  n30 = ScalarParameter() {defaultValue=1, parameterName=SweepSpeed}
  n31 = Multiply(A=Constant() {r=0.3}, B=DynamicParameter() {defaultValue=<rgba 0 1 0 1>, paramNames [Color lerp, Param2, appear, speed]}.speed)
  n32 = ScalarParameter() {parameterName=Use Horizontal lines custom speed}
  n33 = Add(
   A=Power(Base=ComponentMask(None=@n29) {g=true}, Exp=ScalarParameter() {defaultValue=4, parameterName=Horizontal lines power}) {constExponent=4}
   B=Multiply(A=Time(), B=LinearInterpolate(A=@n30, B=@n31, Alpha=@n32))
  )
  n34 = Floor(None=Add(A=Sine(None=@n33) {period=0.2}) {constB=0.03})
  n35 = ComponentMask(None=TextureCoordinate()) {g=true}
  n36 = Panner(
   Coordinate=TextureCoordinate() {uTiling=2}
   Time=LinearInterpolate(
    A=DynamicParameter() {defaultValue=<rgba 0 1 0 1>, paramNames [Color lerp, Param2, appear, speed]}.speed
    B=Time()
    Alpha=ScalarParameter() {parameterName=Use Loop square loop panner}
   )
  ) {speedY=2}
  n37 = Multiply(
   A=Multiply(A=Multiply(A=@n34, B=@n35))
   B=LinearInterpolate(Alpha=TextureSample(UVs=@n36) {texture=/Game/Effects/Fort_Effects/Emotes/Candor/T_Noise_Blur_Texture}.R) {constA=-1, constB=10}
  )
  n38 = TextureCoordinate() {uTiling=60}
  n39 = Add(A=Sine(None=ComponentMask(None=@n38) {r=true})) {constB=0.05}
  n40 = TextureCoordinate()
  n41 = Multiply(A=Multiply(A=Floor(None=@n39), B=ComponentMask(None=@n40) {g=true})) {constB=0.5}
  n42 = TextureCoordinate() {uTiling=2}
  n43 = DynamicParameter() {defaultValue=<rgba 0 1 0 1>, paramNames [Color lerp, Param2, appear, speed]}
  n44 = LinearInterpolate(Alpha=TextureSample(UVs=Panner(Coordinate=@n42, Time=@n43.speed) {speedY=3}) {texture=&T_Noise_Blur_Texture}.R)
  n45 = Multiply(
   A=LinearInterpolate(A=Constant3Vector() {constant=<rgba 0 0.1379 0.4531 1>}, B=Constant3Vector() {constant=<rgba 0 1 0.5318 0>}, Alpha=Saturate(None=@n28))
   B=Add(A=Saturate(None=@n37), B=Multiply(A=@n41, B=@n44)) {constB=0}
  )
  n46 = Multiply(A=Add(A=Add(A=@n26, B=@n14), B=Multiply(A=@n45))) {constB=1.5}
  n47 = MaterialFunctionCall(Vector 3=@n46) {materialFunction=/Engine/Functions/Engine_MaterialFunctions02/Utility/VectorLength}
  n48 = ComponentMask(None=TextureCoordinate()) {g=true}
  n49 = Saturate(None=Multiply(A=Power(Base=@n48) {constExponent=1.7}, B=ScalarParameter() {defaultValue=20, parameterName=TopMask_Multiply}) {constB=20})
  n50 = OneMinus(None=ComponentMask(None=TextureCoordinate()) {g=true})
  n51 = ScalarParameter() {defaultValue=1, parameterName=BottomMask_Power}
  n52 = Saturate(None=Multiply(A=Power(Base=@n50, Exp=@n51) {constExponent=1}) {constB=20})
  n53 = Saturate(None=Multiply(A=Multiply(A=@n49, B=@n52)))
  n54 = Multiply(A=@n1.A, B=LinearInterpolate(A=@n53, B=Multiply(A=@n47.V3 Length, B=@n53), Alpha=ScalarParameter() {parameterName=Use square mask}))
  n55 = ComponentMask(None=TextureCoordinate()) {g=true}
  n56 = LinearInterpolate(Alpha=DynamicParameter() {defaultValue=<rgba 0 1 0 1>, paramNames [Color lerp, Param2, appear, speed]}.appear) {constA=-0.3}
  n57 = Saturate(None=Multiply(A=Subtract(A=@n55, B=@n56) {constB=-1}) {constB=10})
  n58 = Multiply(A=MaterialFunctionCall("FadeSensitivity (optional)"=ScalarParameter() {parameterName=MeshFade}) {materialFunction=&FadeBasedOnViewAngle}.Result) {constB=3}
  DepthFade(Opacity=Multiply(A=Multiply(A=@n54, B=@n57), B=Clamp(None=@n58)), FadeDistance=ScalarParameter() {parameterName=DepthFade})
  n59 = TextureSample(UVs=Panner(
    Coordinate=TextureCoordinate() {vTiling=0.2}
    Time=DynamicParameter() {defaultValue=<rgba 0 1 0 1>, paramNames [Color lerp, Param2, appear, speed]}.speed
   ) {speedY=1}) {texture=/Game/Effects/Fort_Effects/Emotes/Energize_Max/LV4/T_Energize_Color_Noise_blur}
  n60 = Constant() {r=-2}
  n61 = Multiply(
   A=Power(Base=Desaturation(None=@n59.RGB, Fraction=@n60))
   B=DynamicParameter() {defaultValue=<rgba 0 1 0 1>, paramNames [Color lerp, Param2, appear, speed]}.Color lerp
  ) {constB=0}
  MaterialFunctionCall(Emissive=Multiply(A=@n1.RGB, B=Add(A=@n61, B=@n46))) {materialFunction=&MF_FX_EmissiveAdjustments}
  TextureCoordinate() {vTiling=0.6}
 ]
}
