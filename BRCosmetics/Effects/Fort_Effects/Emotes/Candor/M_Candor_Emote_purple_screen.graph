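# Unlit, two-sided additive material for the purple screen effect in the Candor emote, usable on Niagara mesh particles.
# Emissive adds a base color to particle-tinted content, blending tinted scrolling pixel squares with noise-modulated rainbow rays via Use Ray Tex.
# Opacity combines vertical edge fades, particle alpha, optional scanlines, ray and ring-texture masks, and depth fade.
# Dynamic parameters supply panner time (speed) and drive a vertical reveal wipe (appear).
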Material /BRCosmetics/Effects/Fort_Effects/Emotes/Candor/M_Candor_Emote_purple_screen
graph {
 properties {
  bUsedWithNiagaraMeshParticles=true
  blendMode=BLEND_Additive
  pixelDepthOffsetMode=PDOM_Legacy
  shadingModel=MSM_Unlit
  twoSided=true
  usedShadingModels=Unlit
 }
 nodes [
  n1 = TextureCoordinate() {uTiling=3, vTiling=0.5}
  n2 = Constant2Vector() {g=0.05}
  n3 = LinearInterpolate(
   A=Constant() {r=1}
   B=TextureSample(UVs=Add(A=@n1, B=@n2)) {texture=/Game/Effects/Fort_Effects/Textures/T_Smoke_Ring_Mask2}.R
   Alpha=ScalarParameter() {parameterName=Use Tex mask}
  )
  n4 = ParticleColor()
  n5 = TextureCoordinate() {uTiling=3, vTiling=0.1}
  n6 = DynamicParameter() {defaultValue=<rgba 0 1 0 1>, paramNames [Param1,Param2,appear,speed]}
  n7 = Desaturation(
   None=TextureSample(UVs=Panner(Coordinate=@n5, Time=@n6.speed) {speedY=1}) {texture=&T_Tile_RainbowColor}.RGB
   Fraction=Constant() {r=-1}
  )
  n8 = Multiply(A=@n7)
  n9 = Add(B=Add(A=Round(None=@n8), B=@n7))
  n10 = TextureSample(UVs=Panner(Coordinate=TextureCoordinate() {vTiling=0.5}, Time=@n6.speed) {speedY=2}) {samplerType=SAMPLERTYPE_Grayscale, texture=&LowResBlurredNoise}
  n11 = Add(A=Power(Base=LinearInterpolate(Alpha=@n10.R))) {constB=0}
  n12 = Add(B=Multiply(A=Multiply(A=@n9, B=@n11)) {constB=1.5})
  n13 = LinearInterpolate(
   A=Constant() {r=1}
   B=MaterialFunctionCall(Vector 3=@n12) {materialFunction=/Engine/Functions/Engine_MaterialFunctions02/Utility/VectorLength}.V3 Length
   Alpha=ScalarParameter() {parameterName=Use Ray Tex}
  )
  n14 = Multiply(A=ComponentMask(None=TextureCoordinate()) {g=true}, B=ScalarParameter() {defaultValue=5, parameterName=Mask edge intensity}) {constB=5}
  n15 = ComponentMask(None=TextureCoordinate()) {g=true}
  n16 = Multiply(A=Power(Base=OneMinus(None=@n15)) {constExponent=1}) {constB=3}
  n17 = Multiply(A=Multiply(A=Saturate(None=@n14), B=Saturate(None=@n16)))
  n18 = ScalarParameter() {defaultValue=150, parameterName=Scanline Amount}
  n19 = ScalarParameter() {defaultValue=-0.15, parameterName=ScanlineSize}
  n20 = ScalarParameter() {defaultValue=0.5, parameterName=Scanlines Dark}
  n21 = ScalarParameter() {defaultValue=1, parameterName=Scanlines Light}
  n22 = Multiply(
   A=Multiply(A=@n4.A, B=Multiply(A=@n17, B=@n13))
   B=LinearInterpolate(
    A=Constant() {r=1}
    B=MaterialFunctionCall!(
     Number of Scanlines=@n18
     Scanline Speed
     "Scanlines Thickness (Bias)"=@n19
     Scanlines Dark Value=@n20
     Scanlines Light Value=@n21
     Block Opacity
     UVs
     Time
     Block Repeat
    ) {materialFunction=/Game/Effects/Fort_Effects/Effects/Atlas/Materials/MF_Atlas_Scanlines_01}.Result
    Alpha=ScalarParameter() {defaultValue=1, parameterName=Use Scanlines}
   )
  )
  n23 = TextureCoordinate()
  n24 = DynamicParameter() {defaultValue=<rgba 0 1 0 1>, paramNames [Param1,Param2,appear,speed]}
  n25 = Multiply(A=Subtract(A=ComponentMask(None=@n23) {g=true}, B=LinearInterpolate(Alpha=@n24.appear) {constA=-0.3}) {constB=-1}) {constB=10}
  DepthFade(Opacity=Multiply(A=Multiply(A=@n22, B=@n3), B=Saturate(None=@n25)), FadeDistance=ScalarParameter() {parameterName=DepthFade})
  n26 = VectorParameter() {parameterName=Base color, defaultValue=<rgba 0.0321 0 0.0625 0>}
  n27 = TextureSample(UVs=Add(A=TextureCoordinate() {uTiling=3, vTiling=0.7}, B=Constant2Vector() {g=-0.05})) {texture=&T_Smoke_Ring_Mask2}
  n28 = TextureCoordinate() {uTiling=3, vTiling=0.85}
  n29 = ComponentMask(None=@n28) {r=true}
  n30 = Power(Base=ComponentMask(None=@n28) {g=true}) {constExponent=3}
  n31 = TextureSample(UVs=Panner(
    Coordinate=AppendVector(A=@n29, B=@n30)
    Time=DynamicParameter() {defaultValue=<rgba 0 1 0 1>, paramNames [Param1,Param2,appear,speed]}.speed
   ) {speedY=1}) {
   samplerType=SAMPLERTYPE_LinearGrayscale
   texture=/Game/UI/Frontend/ItemShop/Com/vfx/T_CommunityVoting_Activities_Pixels.CommunityVoting_Activities_Pixels
  }
  n32 = Constant() {r=-1}
  n33 = TextureSample(UVs=Panner(Coordinate=TextureCoordinate(), Time=DynamicParameter() {defaultValue=<rgba 0 1 0 1>, paramNames [Param1,Param2,appear,speed]}.speed) {speedY=3}) {samplerType=SAMPLERTYPE_Grayscale, texture=&LowResBlurredNoise}
  n34 = Multiply(A=Add(B=Desaturation(None=@n31.RGB, Fraction=@n32)), B=Power(Base=LinearInterpolate(Alpha=@n33.R)))
  n35 = Multiply(A=Add(A=Multiply(A=@n34, B=@n27.R)) {constB=0}) {constB=1.5}
  n36 = VectorParameter() {parameterName=Square color, defaultValue=<rgba 0 0.3073 0.4 0.2>}
  n37 = Multiply(A=@n4.RGB, B=LinearInterpolate(A=Multiply(A=@n35, B=@n36.RGB), B=@n12, Alpha=ScalarParameter() {parameterName=Use Ray Tex}))
  MaterialFunctionCall(Emissive=Add(B=Add(A=@n26.RGB, B=@n37))) {materialFunction=&MF_FX_EmissiveAdjustments}
 ]
}
